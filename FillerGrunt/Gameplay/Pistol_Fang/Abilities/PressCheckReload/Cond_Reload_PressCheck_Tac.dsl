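; Segmented reload condition for the Pistol_Fang weapon that decides whether the tactical press-check reload applies.
; It passes when the weapon's owner has the Ability.FillerGrunt.Reload.PressCheck.Valid tag and a value from the reload state data is above zero.
; On a pass it also removes all stacks of GE_PressCheckReload from the owner's ability system, so evaluation has a side effect.
; Otherwise, including for an invalid owner, it returns false.
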(fn EvaluateCondition
  (params (ReloadStateData :type Segmented%20Reload%20State%20Data%20Structure%20%28by%20ref%29) (WeaponActor :type Fort%20Weapon%20Object%20Reference))
  (bind _returnvalue (Actor|GetOwner WeaponActor))
  (Utilities|IsValid _returnvalue
    (:Is%20Valid
      (bind _returnvalue_2 (GameplayTags|HasAnyMatchingGameplayTags(Message) _returnvalue (GameplayTags|MakeLiteralGameplayTagContainer "(GameplayTags=((TagName=\"Ability.FillerGrunt.Reload.PressCheck.Valid\")))")))
      (if (and _returnvalue_2 (> (Utilities|Struct|BreakSegmentedReloadStateData ReloadStateData) 0))
        (bind _returnvalue_3 (Class|FortAbilitySystemInterface|GetAbilitySystemComponentDontCreate(Message) _returnvalue))
        (GameplayEffects|RemoveActiveGameplayEffectbySourceEffect _returnvalue_3 "/FillerGrunt/Gameplay/Pistol_Fang/Abilities/PressCheckReload/GE_PressCheckReload.GE_PressCheckReload_C" 0 -1)
        (return true)
        (else
          _)))
    (:Is%20Not%20Valid
      _))
  (return false))
