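; GA_PressCheck_Handler is a gameplay ability that grants the Fang pistol's press-check reload effect on kills.
; On activation it casts the current source weapon to BP_FG_Fang_Pistol and stores it, ending if the cast fails.
; It then keeps listening for Event.Damage.Killed, applying GE_PressCheckReload to the owner on each event.
; The ability stays active as a persistent listener, and its end-ability handler is empty.
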(event Ability|EventActivateAbility
  (bind _asbp_fg_fang_pistol (Utilities|Casting|CastToBP_FG_Fang_Pistol (Ability|GetCurrentSourceWeapon))
    (:then
      (Variables|Default|SetSourceWeapon _asbp_fg_fang_pistol)
      (Ability|Tasks|WaitGameplayEvent "(TagName=\"Event.Damage.Killed\")" 0 false false
        (:EventReceived
          (Ability|ApplyGameplayEffectToOwner "/FillerGrunt/Gameplay/Pistol_Fang/Abilities/PressCheckReload/GE_PressCheckReload.GE_PressCheckReload_C" 0))))
    (:CastFailed
      (Ability|EndAbility))))

(event Ability|EventOnEndAbility
  (params (bWasCancelled :type Boolean)))
